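; Spawner actor that creates an AI bot on its owner's team if friendly, or on fixed team 17 otherwise.
; BeginPlay checks the owner's team is valid, then requests the bot from the AI spawner at its own transform.
; The spawn callback matches the request ID, hides the pawn, switches team and fires a delayed replicated spawn cue.
; The BotPawn RepNotify plays a delayed local res-in cue, then unhides the bot.
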(event EventBeginPlay
  ;[1 1 1] Begin Play - Get Gamemode and then spawn in the bot setting its team based on if it is friendly or not.
  (if (>= (Team|GetActorTeam (Actor|GetOwner)) (Variables|Default|GetStartOfValidTeams))
    (bind _returnvalue (AI|GetAISpawner))
    (EventDispatchers|BindEventtoOnPawnSpawned _returnvalue (EventDispatchers|CreateEvent :!event Custom|OnPawnSpawned))
    (bind _returnvalue_2 (Class|FortAthenaAISpawnerData|CreateComponentListfromClass (Variables|Default|GetAISpawnerBotClass) self))
    (bind _returnvalue_3 (AI|GetAISpawner))
    (bind _returnvalue_4 (Class|AthenaAISpawner|RequestSpawn _returnvalue_3 _returnvalue_2 (Math|Transform|MakeTransform (Transformation|GetActorLocation self) (Transformation|GetActorRotation self)) false -1))
    (Utilities|Array|Add (Variables|Default|GetSpawnRequestIDs) _returnvalue_4)
    (else
      (Development|PrintString "B_Athena_FrenchYedoc_JWFriendly - BeginPlay ABORT spawn due to Invalid Team from Owner"))))

(event Custom|StartBotSpawnInVisuals
  ;[1 1 1] Start Bot Spawn In Visuals - Spawn in the drone and set its life/anim play rate, attach it to the bot and then tell the bot to play its res in FX
  (bind _botpawn (Variables|Default|GetBotPawn)
    (:then
      (Utilities|FlowControl|Delay (Variables|Default|GetDelayBeforeSpawningStartingSpawninFX))
      (GameplayCue|ExecuteGameplayCue%28Burst%29Local _botpawn (Variables|Default|GetBot%20Pawn%20ResIn%20Cue) (Ability|GameplayCue|MakeGameplayCueParameters 0.0))
      (CallFunction|UnhideBotPawn))
    (:else)))

(event Custom|UnhideBotPawn
  ;[1 1 1] Make sure to unhide bot pawn server side, so when someone joins later, bot is replicated as visible (or when creating pawn back from relevancy)
  (Rendering|SetActorHiddenInGame (Variables|Default|GetBotPawn) false))

(event Custom|OnPawnSpawned
  (params (Pawn :type Pawn%20Object%20Reference) (RequestID :type Integer))
  (bind _isfriendly (Variables|Default|GetIsFriendly))
  (bind _returnvalue (Utilities|Array|RemoveItem (Variables|Default|GetSpawnRequestIDs) RequestID))
  (if _returnvalue
    (bind _asfort_player_pawn_athena (Utilities|Casting|CastToFortPlayerPawnAthena Pawn))
    (Rendering|SetActorHiddenInGame _asfort_player_pawn_athena true)
    (bind _output_get (Variables|Default|SetBotPawn _asfort_player_pawn_athena))
    (bind _asfort_athena_aibot_controller (Utilities|Casting|CastToFortAthenaAIBotController (Pawn|GetController _output_get)))
    (Class|FortAthenaAIBotController|SwitchTeam _asfort_athena_aibot_controller (select _isfriendly (Team|GetActorTeam (Actor|GetOwner)) 17))
    (Utilities|FlowControl|Delay (Variables|Default|GetDelayBeforeCue))
    (Ability|ExecuteGameplayCueWithParamsReplicated (Variables|Default|GetGC_Spawn) (Ability|GameplayCue|MakeGameplayCueParameters :NormalizedMagnitude (select _isfriendly 0.0 1.0) :RawMagnitude 0.0 :Location "0,0,0" :Normal "0,0,0" :Instigator _asfort_player_pawn_athena :GameplayEffectLevel 1 :AbilityLevel 1 :bReplicateLocationWhenUsingMinimalRepProxy false))))

(fn OnRep_BotPawn
  (Networking|CanExecuteCosmeticEvents
    (:True
      (CallFunction|StartBotSpawninVisuals))))

(fn SetBotWeaponsHiddenInGame
  (params (BotPawn :type Fort%20Pawn%20Object%20Reference) (NewHidden :type Boolean) (bPropagateToChildren :type Boolean))
  (for _array_element (Visual|GetWeaponMeshes (Variables|/Script/FortniteGame.FortPawn|GetCurrentWeapon BotPawn))
    (Development|SetHiddeninGame _array_element NewHidden bPropagateToChildren)))

(fn ConstructionScript)
