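# Unlit translucent particle-sprite material for a chevron marker in the FrenchYedoc item folder, drawn from the squad generic indicator texture.
# A world-offset calculation enlarges the sprite around its particle center as camera distance grows, keeping it readable from afar.
# It also adds a vertical offset from the WorldOffsetZ dynamic parameter plus a tunable distance bias.
# The texture color passes through MF_FX_EmissiveAdjustments, and responsive AA is enabled.
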Material /Game/Athena/Items/Gameplay/FrenchYedoc/M_FrenchYedoc_Chevron
graph {
 properties {
  bEnableResponsiveAA=true
  bUsedWithParticleSprites=true
  blendMode=BLEND_Translucent
  pixelDepthOffsetMode=PDOM_Legacy
  shadingModel=MSM_Unlit
  usedShadingModels=Unlit
 }
 nodes [
  n1 = DynamicParameter() {paramNames [WorldOffsetZ,Param2,Param3,Param4]}
  n2 = ParticlePositionWS()
  n3 = Distance(A=@n2, B=CameraPositionWS())
  n4 = Divide(A=@n3, B=ScalarParameter() {defaultValue=50, parameterName=WorldOffsetZDistanceBiasDivisor}) {constB=1000}
  n5 = WorldPosition()
  Add(
   A=MaterialFunctionCall(B=Add(A=@n1.WorldOffsetZ, B=@n4)) {materialFunction=&AppendMany}.RGB
   B=Multiply(A=Subtract(A=@n5.XYZ, B=@n2), B=Divide(A=@n3) {constB=1000})
  )
  MaterialFunctionCall(Emissive=TextureSample() {texture=/Game/Athena/HUD/Art/squadArt/T_UI_SquadGenericIndicator}.RGB) {materialFunction=&MF_FX_EmissiveAdjustments}
 ]
}
